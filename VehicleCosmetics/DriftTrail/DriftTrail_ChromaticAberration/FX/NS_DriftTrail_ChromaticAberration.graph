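# Niagara system for the Chromatic Aberration drift trail vehicle cosmetic, combining wheel-wrap meshes with ribbon trails.
# Wheel-wrap emitters burst mesh particles pinned to skeletal mesh sockets, partitioned and scaled per wheel, with tire-FX masking.
# Paired-ribbon emitters stream backward from the car, easing off inherited movement over lifetime.
# Painted-variant overrides color up to four renderer channels, using per-paint life curves alongside fixed teal and red channels.
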NiagaraSystem /VehicleCosmetics/DriftTrail/DriftTrail_ChromaticAberration/FX/NS_DriftTrail_ChromaticAberration
dependencies {
 usedDataInterfaces [
  &NiagaraDataInterfaceCurve
  &NiagaraDataInterfaceColorCurve
  &NiagaraDataInterfaceArrayFloat3
  &NiagaraDataInterfaceSkeletalMesh
 ]
 usedDynamicInputs []
 usedModules [
  &self:SystemSpawnScript.SystemScriptSource.SystemScriptGraph.NiagaraNodeAssignment_0.SetVariables_#1
  &self:SystemSpawnScript.SystemScriptSource.SystemScriptGraph.NiagaraNodeAssignment_1.SetVariables_#2
  &FN_UnpackVehicleFlags
  &SystemState
  /Game/Effects/Niagara/Modules/Vehicle/FN_UpdateVehicleAttribs
  /Game/Effects/Niagara/Modules/Vehicle/FN_VehicleFlags_Cheatcodes
  &self:SystemSpawnScript.SystemScriptSource.SystemScriptGraph.NiagaraNodeAssignment_2.SetVariables_#3
 ]
}
emitters {
 NE_WheelWrap {
  emitterUpdateScript [
   &EmitterState(
    Inactive Response=::NewEnumerator1
    Life Cycle Mode=::NewEnumerator1
    Loop Behavior=::NewEnumerator1
    Loop Duration Mode=::NewEnumerator1
    UseLoopDelay=false
   )
   &SpawnBurst_Instantaneous(Spawn Count=2, Use Spawn Probability=false, UseLoopCountLimit=false)
   &SpawnBurst_Instantaneous(Spawn Count=2, Spawn Time=0.3667, Use Spawn Probability=false, UseLoopCountLimit=false) {module=SpawnBurst_Instantaneous001}
  ]
  particleSpawnScript [
   &VC_PaintedVariantsOverride(
    DefaultColor=<rgba 0.0027 0.0003 0.0578 1>
    R2 DefaultColor:LinearColor=<rgba 1 0.7991 0.3712 1>
    R3 Black:LinearColor=<rgba 0.1719 0.1719 0.1719 1>
    R3 BurntSienna:LinearColor=<rgba 0.3177 0.1227 0.0556 1>
    R3 Cobalt:LinearColor=<rgba 0.0165 0.0249 0.474 1>
    R3 Crimson:LinearColor=<rgba 1 0.0365 0.0626 1>
    R3 DefaultColor:LinearColor=<rgba 3 2.008 1.14 1>
    R3 ForestGreen:LinearColor=<rgba 0 0.4427 0.0224 1>
    R3 Grey:LinearColor=<rgba 0.4271 0.4271 0.4271 1>
    R3 Lime:LinearColor=<rgba 0.5453 0.8698 0.0076 1>
    R3 Orange:LinearColor=<rgba 0.7917 0.3402 0 1>
    R3 Pink:LinearColor=<rgba 1 0.0781 0.7743 1>
    R3 Purple:LinearColor=<rgba 0.3242 0.0078 0.7396 1>
    R3 Saffron:LinearColor=<rgba 0.8542 0.6632 0.0356 1>
    R3 SkyBlue:LinearColor=<rgba 0.0485 0.5326 0.776 1>
    R3 TitaniumWhite:LinearColor=<rgba 2 2 2 1>
    Use R2 Painted Overrides=true
    Use R3 Painted Overrides=true
   ) {enabled=false}
   &InitializeParticle(
    Color=<rgba 0.0568 0.025 0.0578 1>
    Lifetime=$Emitter.CurrentLoopDuration
    Mesh Scale Mode=::NewEnumerator4
    Sprite Rotation Mode=::NewEnumerator2
    Sprite Size Mode=::NewEnumerator3
    UsePositionOffset=false
    Write Material Random=true
   )
   &PartitionParticles()
   &InitialMeshOrientation(
    Mesh Orientation Mode=::NewEnumerator7
    Rotation=&SelectVectorFromArray(
     Direct Array Index=$Particles.PartitionParticles.Partition
     Vector Selection Array=DataInterface() {propertyValues {FloatData [<0 0 0>, <0 0 0>]}}
    )
    Use Rotation Vector=true
   )
   &SkeletalMeshLocation(
    Bone / Socket Index=&Add_Integer(A=$Particles.PartitionParticles.Partition, B=4)
    Bone / Socket Sampling Mode=::NewEnumerator5
    Interpolate Mesh Sampling=false
    MaskSpawnGroup=false
    Mesh Sampling Type=::NewEnumerator6
    Skeletal Mesh=$System.SkeletalMesh
   )
  ]
  particleUpdateScript [
   &FN_DisableTireFX(
    Mask Index Attribute:NiagaraInt32=$Particles.PartitionParticles.Partition
    Mask Index FL=2
    Mask Index FR=3
    Mask Index RL=0
    Mask Index RR=1
   )
   &VC_WheelWrapScale(Wheel Index Attribute:NiagaraInt32=&Add_Integer(A=$Particles.PartitionParticles.Partition, B=2))
   &ParticleState(Kill Particles When Lifetime Has Elapsed=false, Let Infinitely Lived Particles Die When Emitter Deactivates=true)
   &UpdateMeshOrientation(Rotation Coordinate Space=::NewEnumerator0, Rotation Rate=0.5, Rotation Vector=<0 0.8 0>) {enabled=false}
   &SkeletalMeshLocation(
    Bone / Socket Index=&Add_Integer(A=$Particles.PartitionParticles.Partition, B=4)
    Bone / Socket Sampling Mode=::NewEnumerator5
    MaskSpawnGroup=false
    Mesh Sampling Type=::NewEnumerator6
    Orientation Sampling=::NewEnumerator1
    Sampled Position Offset=&SelectVectorFromArray(
     Direct Array Index=$Particles.PartitionParticles.Partition
     Vector Selection Array=DataInterface() {propertyValues {FloatData [<0 0 0>, <0 0 0>]}}
    )
    Skeletal Mesh=$System.SkeletalMesh
   ) {module=SkeletalMeshLocation001}
   &SolveForcesAndVelocity(Clamp Velocity=false, Limit Acceleration=false)
   &VC_PaintedVariantsOverride(
    Black=<rgba 0.1458 0.1458 0.1458 1>
    Orange=<rgba 0.6724 0.0883 0.2836 1>
    R1 DefaultAlpha=0.9
    R1 DefaultColor=<rgba 0 0.0686 0.1771 1>
    R2 Black:LinearColor=<rgba 0 0.7991 0.5089 1>
    R2 BurntSienna:LinearColor=<rgba 0 0.7991 0.5089 1>
    R2 Cobalt:LinearColor=<rgba 0 0.7991 0.5089 1>
    R2 Crimson:LinearColor=<rgba 0 0.7991 0.5089 1>
    R2 DefaultAlpha=0.1
    R2 DefaultColor:LinearColor=&ColorFromCurve(ColorCurve=DataInterface() {propertyValues {
       AlphaCurve {keys [<at 0.0449 3>,<at 0.4797 3>,<at 1 3>]}
       BlueCurve {keys [<at 0.4885 0.5089>]}
       GreenCurve {keys [<at 0.4885 0.7991>]}
       RedCurve {keys [<at 0.4885 0>]}
      }})
    R2 ForestGreen:LinearColor=<rgba 0 0.7991 0.5089 1>
    R2 Grey:LinearColor=<rgba 0 0.7991 0.5089 1>
    R2 Lime:LinearColor=<rgba 0 0.7991 0.5089 1>
    R2 Orange:LinearColor=<rgba 0 0.7991 0.5089 1>
    R2 Pink:LinearColor=<rgba 0 0.7991 0.5089 1>
    R2 Purple:LinearColor=<rgba 0 0.7991 0.5089 1>
    R2 Saffron:LinearColor=<rgba 0 0.7991 0.5089 1>
    R2 SkyBlue:LinearColor=<rgba 0 0.7991 0.5089 1>
    R2 TitaniumWhite:LinearColor=<rgba 0 0.7991 0.5089 1>
    R3 Black:LinearColor=<rgba 0.624 0 0.0194 0.8196>
    R3 BurntSienna:LinearColor=<rgba 0.624 0 0.0194 0.8196>
    R3 Cobalt:LinearColor=<rgba 0.624 0 0.0194 0.8196>
    R3 Crimson:LinearColor=<rgba 0.624 0 0.0194 0.8196>
    R3 DefaultAlpha=4
    R3 DefaultColor:LinearColor=<rgba 0.624 0 0.0194 0.8196>
    R3 ForestGreen:LinearColor=<rgba 0.624 0 0.0194 0.8196>
    R3 Grey:LinearColor=<rgba 0.624 0 0.0194 0.8196>
    R3 Lime:LinearColor=<rgba 0.624 0 0.0194 0.8196>
    R3 Orange:LinearColor=<rgba 0.624 0 0.0194 0.8196>
    R3 Pink:LinearColor=<rgba 0.624 0 0.0194 0.8196>
    R3 Purple:LinearColor=<rgba 0.624 0 0.0194 0.8196>
    R3 Saffron:LinearColor=<rgba 0.624 0 0.0194 0.8196>
    R3 SkyBlue:LinearColor=<rgba 0.624 0 0.0194 0.8196>
    R3 TitaniumWhite:LinearColor=<rgba 0.624 0 0.0194 0.8196>
    R4 Black:LinearColor=<rgba 0.4688 0.4688 0.4688 1>
    R4 BurntSienna:LinearColor=<rgba 0.5573 0.1721 0.0687 1>
    R4 Cobalt:LinearColor=<rgba 0.3177 0.3303 1 1>
    R4 Crimson:LinearColor=<rgba 0.7396 0.2658 0.2786 1>
    R4 DefaultColor:LinearColor=<rgba 1.375 1.4491 1.5 0.8667>
    R4 ForestGreen:LinearColor=<rgba 0.2865 0.8594 0.3154 1>
    R4 Grey:LinearColor=<rgba 0.6198 0.6198 0.6198 1>
    R4 Lime:LinearColor=<rgba 0.6051 0.8802 0.3301 1>
    R4 Orange:LinearColor=<rgba 0.9219 0.4904 0.2593 1>
    R4 Pink:LinearColor=<rgba 0.8854 0.3874 0.7043 1>
    R4 Purple:LinearColor=<rgba 0.5347 0.2905 0.8854 1>
    R4 Saffron:LinearColor=<rgba 1.2 1.0469 0.5438 1>
    R4 SkyBlue:LinearColor=<rgba 0.5443 0.914 1.1 1>
    R4 TitaniumWhite:LinearColor=<rgba 1 1 1 1>
    Use R1 Painted Overrides=true
    Use R2 Painted Overrides=true
    Use R3 Painted Overrides=true
    Use R4 Painted Overrides=true
   ) {module=VC_PaintedVariantsOverride002}
  ]
  rendererClasses [&NiagaraMeshRendererProperties]
  renderers [
   {rendererClass=&NiagaraMeshRendererProperties, rendererIndex=0}
   {rendererClass=&NiagaraMeshRendererProperties, rendererIndex=1}
   {rendererClass=&NiagaraMeshRendererProperties, rendererIndex=2}
   {rendererClass=&NiagaraMeshRendererProperties, rendererIndex=3}
   {rendererClass=&NiagaraMeshRendererProperties, rendererIndex=4}
  ]
 }
 NE_WheelWrap001 {
  emitterUpdateScript [
   &EmitterState(
    Inactive Response=::NewEnumerator1
    Life Cycle Mode=::NewEnumerator1
    Loop Behavior=::NewEnumerator1
    Loop Duration Mode=::NewEnumerator1
    UseLoopDelay=false
   )
   &SpawnBurst_Instantaneous(Spawn Count=2, Use Spawn Probability=false, UseLoopCountLimit=false)
  ]
  particleSpawnScript [
   &VC_PaintedVariantsOverride(
    DefaultColor=<rgba 0.0027 0.0003 0.0578 1>
    R2 DefaultColor:LinearColor=<rgba 1 0.7991 0.3712 1>
    R3 Black:LinearColor=<rgba 0.1719 0.1719 0.1719 1>
    R3 BurntSienna:LinearColor=<rgba 0.3177 0.1227 0.0556 1>
    R3 Cobalt:LinearColor=<rgba 0.0165 0.0249 0.474 1>
    R3 Crimson:LinearColor=<rgba 1 0.0365 0.0626 1>
    R3 DefaultColor:LinearColor=<rgba 3 2.008 1.14 1>
    R3 ForestGreen:LinearColor=<rgba 0 0.4427 0.0224 1>
    R3 Grey:LinearColor=<rgba 0.4271 0.4271 0.4271 1>
    R3 Lime:LinearColor=<rgba 0.5453 0.8698 0.0076 1>
    R3 Orange:LinearColor=<rgba 0.7917 0.3402 0 1>
    R3 Pink:LinearColor=<rgba 1 0.0781 0.7743 1>
    R3 Purple:LinearColor=<rgba 0.3242 0.0078 0.7396 1>
    R3 Saffron:LinearColor=<rgba 0.8542 0.6632 0.0356 1>
    R3 SkyBlue:LinearColor=<rgba 0.0485 0.5326 0.776 1>
    R3 TitaniumWhite:LinearColor=<rgba 2 2 2 1>
    Use R2 Painted Overrides=true
    Use R3 Painted Overrides=true
   ) {enabled=false}
   &InitializeParticle(
    Color=<rgba 0.0027 0.0003 0.0578 1>
    Lifetime=$Emitter.CurrentLoopDuration
    Mesh Scale Mode=::NewEnumerator4
    Sprite Rotation Mode=::NewEnumerator2
    Sprite Size Mode=::NewEnumerator3
    UsePositionOffset=false
    Write Material Random=true
   )
   &PartitionParticles()
   &InitialMeshOrientation(
    Mesh Orientation Mode=::NewEnumerator7
    Rotation=&SelectVectorFromArray(
     Direct Array Index=$Particles.PartitionParticles.Partition
     Vector Selection Array=DataInterface() {propertyValues {FloatData [<0 0 0>, <-0.5 0 0>]}}
    )
    Use Rotation Vector=true
   )
   &SkeletalMeshLocation(
    Bone / Socket Index=&Add_Integer(A=$Particles.PartitionParticles.Partition, B=4)
    Bone / Socket Sampling Mode=::NewEnumerator5
    Interpolate Mesh Sampling=false
    MaskSpawnGroup=false
    Mesh Sampling Type=::NewEnumerator6
    Skeletal Mesh=$System.SkeletalMesh
   )
  ]
  particleUpdateScript [
   &FN_DisableTireFX(
    Mask Index Attribute:NiagaraInt32=$Particles.PartitionParticles.Partition
    Mask Index FL=2
    Mask Index FR=3
    Mask Index RL=0
    Mask Index RR=1
   )
   &VC_WheelWrapScale(Wheel Index Attribute:NiagaraInt32=&Add_Integer(A=$Particles.PartitionParticles.Partition, B=2))
   &ParticleState(Kill Particles When Lifetime Has Elapsed=false, Let Infinitely Lived Particles Die When Emitter Deactivates=true)
   &UpdateMeshOrientation(Rotation Coordinate Space=::NewEnumerator0, Rotation Rate=0.5, Rotation Vector=<0 0.8 0>)
   &SkeletalMeshLocation(
    Bone / Socket Index=&Add_Integer(A=$Particles.PartitionParticles.Partition, B=4)
    Bone / Socket Sampling Mode=::NewEnumerator5
    MaskSpawnGroup=false
    Mesh Sampling Type=::NewEnumerator6
    Orientation Sampling=::NewEnumerator1
    Sampled Position Offset=&SelectVectorFromArray(
     Direct Array Index=$Particles.PartitionParticles.Partition
     Vector Selection Array=DataInterface() {propertyValues {FloatData [<0 0 0>, <0 0 0>]}}
    )
    Skeletal Mesh=$System.SkeletalMesh
   ) {module=SkeletalMeshLocation001}
   &SolveForcesAndVelocity(Clamp Velocity=false, Limit Acceleration=false)
   &VC_PaintedVariantsOverride(
    Black=<rgba 0.1458 0.1458 0.1458 1>
    BurntSienna=<rgba 0.0312 0.0096 0.0039 1>
    Cobalt=<rgba 0 0.0014 0.0781 1>
    Grey=<rgba 0.1042 0.1042 0.1042 1>
    Orange=<rgba 0.6724 0.0883 0.2836 1>
    R1 Cobalt:LinearColor=<rgba 0 0.005 0.2708 1>
    R1 Crimson:LinearColor=<rgba 0.224 0 0.0061 1>
    R1 DefaultAlpha=0.9
    R1 DefaultColor=<rgba 0.002 0.0426 0.1146 1>
    R1 ForestGreen:LinearColor=<rgba 0 0.1458 0.0074 1>
    R1 Grey:LinearColor=<rgba 0.0469 0.0469 0.0469 1>
    R1 Lime:LinearColor=<rgba 0.1302 0.2604 0 1>
    R1 Orange:LinearColor=<rgba 0.2917 0.0271 0 1>
    R1 Pink:LinearColor=<rgba 0.25 0.0299 0.17 1>
    R1 Purple:LinearColor=<rgba 0.0257 0 0.0625 1>
    R1 Saffron:LinearColor=<rgba 0.2812 0.2156 0 1>
    R1 SkyBlue:LinearColor=<rgba 0 0.0762 0.1146 1>
    R1 TitaniumWhite:LinearColor=<rgba 0.2135 0.2135 0.2135 1>
    R2 Black:LinearColor=<rgba 0 0.7991 0.5089 1>
    R2 BurntSienna:LinearColor=<rgba 0 0.7991 0.5089 1>
    R2 Cobalt:LinearColor=<rgba 0 0.7991 0.5089 1>
    R2 Crimson:LinearColor=<rgba 0 0.7991 0.5089 1>
    R2 DefaultAlpha=0.1
    R2 DefaultColor:LinearColor=&ColorFromCurve(ColorCurve=DataInterface() {propertyValues {
       AlphaCurve {keys [<at 0.0449 3>,<at 0.4797 3>,<at 1 3>]}
       BlueCurve {keys [<at 0.4885 0.5089>]}
       GreenCurve {keys [<at 0.4885 0.7991>]}
       RedCurve {keys [<at 0.4885 0>]}
      }})
    R2 ForestGreen:LinearColor=<rgba 0 0.7991 0.5089 1>
    R2 Grey:LinearColor=<rgba 0 0.7991 0.5089 1>
    R2 Lime:LinearColor=<rgba 0 0.7991 0.5089 1>
    R2 Orange:LinearColor=<rgba 0 0.7991 0.5089 1>
    R2 Pink:LinearColor=<rgba 0 0.7991 0.5089 1>
    R2 Purple:LinearColor=<rgba 0 0.7991 0.5089 1>
    R2 Saffron:LinearColor=<rgba 0 0.7991 0.5089 1>
    R2 SkyBlue:LinearColor=<rgba 0 0.7991 0.5089 1>
    R2 TitaniumWhite:LinearColor=<rgba 0 0.7991 0.5089 1>
    R3 Black:LinearColor=<rgba 0.624 0 0.0194 0.8196>
    R3 BurntSienna:LinearColor=<rgba 0.624 0 0.0194 0.8196>
    R3 Cobalt:LinearColor=<rgba 0.624 0 0.0194 0.8196>
    R3 Crimson:LinearColor=<rgba 0.624 0 0.0194 0.8196>
    R3 DefaultAlpha=4
    R3 DefaultColor:LinearColor=<rgba 0.624 0 0.0194 0.8196>
    R3 ForestGreen:LinearColor=<rgba 0.624 0 0.0194 0.8196>
    R3 Grey:LinearColor=<rgba 0.624 0 0.0194 0.8196>
    R3 Lime:LinearColor=<rgba 0.624 0 0.0194 0.8196>
    R3 Orange:LinearColor=<rgba 0.624 0 0.0194 0.8196>
    R3 Pink:LinearColor=<rgba 0.624 0 0.0194 0.8196>
    R3 Purple:LinearColor=<rgba 0.624 0 0.0194 0.8196>
    R3 Saffron:LinearColor=<rgba 0.624 0 0.0194 0.8196>
    R3 SkyBlue:LinearColor=<rgba 0.624 0 0.0194 0.8196>
    R3 TitaniumWhite:LinearColor=<rgba 0.624 0 0.0194 0.8196>
    R4 Black:LinearColor=<rgba 0.0417 0.0417 0.0417 0.8>
    R4 BurntSienna:LinearColor=<rgba 0.0208 0.0066 0 1>
    R4 Cobalt:LinearColor=<rgba 0 0.0053 0.2865 1>
    R4 Crimson:LinearColor=<rgba 0.4688 0.0073 0.0198 1>
    R4 DefaultColor:LinearColor=&ColorFromCurve(ColorCurve=DataInterface() {propertyValues {
       AlphaCurve {keys [<at 0 1 RCIM_Cubic>,<at 1 0.74 RCIM_Cubic>]}
       BlueCurve {keys [
         <at 0.0519 0.4167>
         {arriveTangent=1.2947, interpMode=RCIM_Cubic, leaveTangent=1.2947, time=0.2632, value=0.5391}
         <at 0.4445 0.9249>
         <at 0.5131 0.9323>
         <at 0.8845 2 RCIM_Cubic>
        ]}
       GreenCurve {keys [
         <at 0.0519 0.2076>
         {arriveTangent=1.9676, interpMode=RCIM_Cubic, leaveTangent=1.9676, time=0.2632, value=1.431}
         <at 0.4445 0.98>
         <at 0.5131 0.8777>
         <at 0.8845 0.3046 RCIM_Cubic>
        ]}
       RedCurve {keys [
         <at 0.0519 0.0129>
         {arriveTangent=1.6444, interpMode=RCIM_Cubic, leaveTangent=1.6444, time=0.2632, value=1.5}
         <at 0.4445 0.6584>
         <at 0.5131 0>
         <at 0.8845 0.0328 RCIM_Cubic>
        ]}
      }})
    R4 ForestGreen:LinearColor=<rgba 0 0.2344 0.0118 1>
    R4 Grey:LinearColor=<rgba 0.0885 0.0885 0.0885 1>
    R4 Lime:LinearColor=<rgba 0.2943 0.5885 0 1>
    R4 Orange:LinearColor=<rgba 0.8087 0.0435 0 1>
    R4 Pink:LinearColor=<rgba 0.6094 0.054 0.4074 1>
    R4 Purple:LinearColor=<rgba 0.1244 0 0.5217 1>
    R4 Saffron:LinearColor=<rgba 0.6348 0.4867 0 1>
    R4 SkyBlue:LinearColor=<rgba 0 0.2253 0.3385 1>
    R4 TitaniumWhite:LinearColor=<rgba 0.5521 0.5521 0.5521 1>
    Use R1 Painted Overrides=true
    Use R2 Painted Overrides=true
    Use R3 Painted Overrides=true
    Use R4 Painted Overrides=true
   ) {module=VC_PaintedVariantsOverride002}
  ]
  rendererClasses [&NiagaraMeshRendererProperties]
  renderers [
   {rendererClass=&NiagaraMeshRendererProperties, rendererIndex=0}
   {rendererClass=&NiagaraMeshRendererProperties, rendererIndex=1}
   {rendererClass=&NiagaraMeshRendererProperties, rendererIndex=2}
   {rendererClass=&NiagaraMeshRendererProperties, rendererIndex=3}
  ]
 }
 Trail_Chromatic {
  emitterUpdateScript [
   &EmitterState()
   &SpawnRate(
    SpawnRate=&Multiply_Float(
     A=$Engine.Emitter.SpawnCountScale
     B=/VehicleCosmetics/FX/Niagara/Modules/VC_Mask_Floats(Coro Value:NiagaraFloat=25, FrontEnd Value:NiagaraFloat=0.5, FN Value:NiagaraFloat=0.5)
    )
    Use Spawn Probability=false
   ) {module=SpawnRate001}
   &VC_RateQualityHandler(Epic Spawn Rate=40, High Spawn Rate=40, Low Spawn Rate=40, Medium Spawn Rate=40, Use Custom Rate?=true)
   &VC_SpawnRibbon_v1(Number of Ribbons=2, Spawn Time=-1, SpawnRate=$Emitter.Spawn Rate)
  ]
  particleSpawnScript [
   &VC_PaintedVariantsOverride() {enabled=false}
   &VC_InitializeRibbon_v1(
    Color=$Particles.Color
    Custom Offset=true
    Enable Skeletal Position=true
    Lifetime=&MergeHelper_FloatSwitch(0=1.1, 1=1.1, 2=1.1, 3=1.1, Index=$User.GameModeID, Num=::NewEnumerator2)
    Local Vector Offset=&AddVectorToPosition(Position=<-30 0 -30>, Vector=Linked() {linkedVariable {name=User.DTCarOffset, type=NiagaraPosition}})
    Offset Amount=1
    Ribbon Width=55
   )
   &AddVelocity(Velocity=<-1000 0 25>) {module=AddVelocity001}
   /VehicleCosmetics/FX/VC_FXParents/Modules/DLR_ProductPreviews(GameModeID:NiagaraInt32=$System.GameModeID, bPreviewVehicleSpeed:NiagaraBool=$System.bPreviewVehicleSpeed)
  ]
  particleUpdateScript [
   &FN_DisableTireFX(Mask Index Attribute:NiagaraInt32=$Particles.SpawnGroup, Mask Index FL=2, Mask Index FR=3, Mask Index RL=0, Mask Index RR=1)
   &InheritSourceMovement(
    Applied Position Scale=&Multiply_VectorByFloat(Vector=<1 1 1>, Float=1)
    ApplyForce=false
    ApplyPosition=true
    ApplyPositionFalloff=false
   )
   &ParticleState()
   &InheritSourceMovement(
    Applied Position Scale=&Multiply_VectorByFloat(
     Vector=<1 1 1>
     Float=&FloatFromCurve(FloatCurve=DataInterface() {propertyValues {Curve {
         defaultValue=1
         keys [
          {arriveTangent=-1.4003, interpMode=RCIM_Cubic, leaveTangent=-1.4003, tangentMode=RCTM_User, time=0, value=1}
          {arriveTangent=-0.1632, interpMode=RCIM_Cubic, leaveTangent=-0.1632, tangentMode=RCTM_User, time=1, value=0}
         ]
        }}})
    )
    ApplyForce=false
    ApplyPosition=true
    ApplyPositionFalloff=false
   ) {module=InheritSourceMovement001}
   &SolveForcesAndVelocity(Clamp Velocity=false, Limit Acceleration=false)
   &VC_UpdateRibbon_v1(Enable Skeletal Position=true)
   &VC_PaintedVariantsOverride(
    Black=<rgba 0.1458 0.1458 0.1458 1>
    Orange=<rgba 0.6724 0.0883 0.2836 1>
    R1 Black:LinearColor=&ColorFromCurve(ColorCurve=DataInterface() {propertyValues {
       AlphaCurve {keys [
         {arriveTangentWeight=0.3333, interpMode=RCIM_Cubic, leaveTangentWeight=0.3333, tangentMode=RCTM_User, time=0, value=4}
         {arriveTangentWeight=0.3333, interpMode=RCIM_Cubic, leaveTangentWeight=0.3333, tangentMode=RCTM_User, time=0.9353, value=0.8}
        ]}
       BlueCurve {keys [
         {arriveTangentWeight=0.1904, interpMode=RCIM_Cubic, leaveTangentWeight=0.1904, tangentMode=RCTM_User, time=0, value=0.3005}
         {tangentMode=RCTM_User, time=0.9333, value=0.0723}
        ]}
       GreenCurve {keys [
         {arriveTangentWeight=0.1904, interpMode=RCIM_Cubic, leaveTangentWeight=0.1904, tangentMode=RCTM_User, time=0, value=0.3005}
         {tangentMode=RCTM_User, time=0.9333, value=0.0723}
        ]}
       RedCurve {keys [
         {arriveTangentWeight=0.1904, interpMode=RCIM_Cubic, leaveTangentWeight=0.1904, tangentMode=RCTM_User, time=0, value=0.3005}
         {tangentMode=RCTM_User, time=0.9333, value=0.0723}
        ]}
      }})
    R1 BurntSienna:LinearColor=&ColorFromCurve(ColorCurve=DataInterface() {propertyValues {
       AlphaCurve {keys [<at 0 4 RCIM_Cubic>,<at 0.2166 2>,<at 0.9253 0.5 RCIM_Cubic>]}
       BlueCurve {keys [<at 0 0>,<at 0.6066 0>]}
       GreenCurve {keys [<at 0 0.2086>,<at 0.6066 0.1046>]}
       RedCurve {keys [<at 0 0.3372>,<at 0.6066 0.1651>]}
      }})
    R1 Cobalt:LinearColor=&ColorFromCurve(ColorCurve=DataInterface() {propertyValues {
       AlphaCurve {keys [<at 0 4 RCIM_Cubic>,<at 0.7138 0.8 RCIM_Cubic>]}
       BlueCurve {keys [<at 0 0.9047 RCIM_Cubic>,<at 0.7058 0.4072>]}
       GreenCurve {keys [<at 0 0.0497 RCIM_Cubic>,<at 0.7058 0.0075>]}
       RedCurve {keys [<at 0 0.0331 RCIM_Cubic>,<at 0.7058 0>]}
      }})
    R1 Crimson:LinearColor=&ColorFromCurve(ColorCurve=DataInterface() {propertyValues {
       AlphaCurve {keys [<at 0 4 RCIM_Cubic>,<at 0.9109 0.5 RCIM_Cubic>]}
       BlueCurve {keys [<at 0.0162 0.007>,<at 0.624 0.0369>]}
       GreenCurve {keys [<at 0.0162 0.007>,<at 0.624 0.0252>]}
       RedCurve {keys [<at 0.0162 0.9047>,<at 0.624 0.4452>]}
      }})
    R1 DefaultAlpha=0.9
    R1 DefaultColor=&ColorFromCurve(ColorCurve=DataInterface() {propertyValues {
       AlphaCurve {keys [<at 0 4 RCIM_Cubic>,<at 0.8423 0.9976>,<at 1 0 RCIM_Cubic>]}
       BlueCurve {keys [<at 0.2544 1>,<at 0.455 1>,<at 0.6592 0.5729>,<at 0.8458 0.625>]}
       GreenCurve {keys [<at 0.2544 0.947>,<at 0.455 0.9298>,<at 0.6592 0.3571>,<at 0.8458 0.1586>]}
       RedCurve {keys [<at 0.2544 0.9293>,<at 0.455 0>,<at 0.6592 0.0027>,<at 0.8458 0>]}
      }})
    R1 ForestGreen:LinearColor=&ColorFromCurve(ColorCurve=DataInterface() {propertyValues {
       AlphaCurve {keys [<at 0 4 RCIM_Cubic>,<at 0.9039 0.5>]}
       BlueCurve {keys [<at 0.0075 0.0887>,<at 0.617 0.0423>]}
       GreenCurve {keys [<at 0.0075 0.855>,<at 0.617 0.3095>]}
       RedCurve {keys [<at 0.0075 0.0497>,<at 0.617 0.0284>]}
      }})
    R1 Grey:LinearColor=&ColorFromCurve(ColorCurve=DataInterface() {propertyValues {
       AlphaCurve {keys [<at 0 4 RCIM_Cubic>,<at 0.9092 0.5 RCIM_Cubic>]}
       BlueCurve {keys [<at 0 0.491>,<at 0.5743 0.2232>]}
       GreenCurve {keys [<at 0 0.491>,<at 0.5743 0.2232>]}
       RedCurve {keys [<at 0 0.491>,<at 0.5743 0.2232>]}
      }})
    R1 Lime:LinearColor=&ColorFromCurve(ColorCurve=DataInterface() {propertyValues {
       AlphaCurve {keys [<at 0 4 RCIM_Cubic>,<at 0.9092 0.5 RCIM_Cubic>]}
       BlueCurve {keys [<at 0 0.0075>,<at 0.3248 0.0075>,<at 0.705 0.0027>]}
       GreenCurve {keys [<at 0 0.9911>,<at 0.3248 0.9911>,<at 0.705 0.3613>]}
       RedCurve {keys [<at 0 0.5647>,<at 0.3248 0.5647>,<at 0.705 0.2051>]}
      }})
    R1 Orange:LinearColor=&ColorFromCurve(ColorCurve=DataInterface() {propertyValues {
       AlphaCurve {keys [<at 0 3 RCIM_Cubic>,<at 0.9213 0.5>]}
       BlueCurve {keys [<at 0 0.0331 RCIM_Cubic>,<at 0.7384 0.0232>]}
       GreenCurve {keys [<at 0 0.3916 RCIM_Cubic>,<at 0.7384 0.2423>]}
       RedCurve {keys [<at 0 0.9131 RCIM_Cubic>,<at 0.7384 0.8148>]}
      }})
    R1 Pink:LinearColor=&ColorFromCurve(ColorCurve=DataInterface() {propertyValues {
       AlphaCurve {keys [<at 0 3 RCIM_Cubic>,<at 0.9277 0.5>]}
       BlueCurve {keys [<at 0 0.9823>,<at 0.7226 0.4342>]}
       GreenCurve {keys [<at 0 0.2159>,<at 0.7226 0.1912>]}
       RedCurve {keys [<at 0 0.9047>,<at 0.7226 0.5776>]}
      }})
    R1 Purple:LinearColor=&ColorFromCurve(ColorCurve=DataInterface() {propertyValues {
       AlphaCurve {keys [<at 0 3 RCIM_Cubic>,<at 0.9011 0.5>]}
       BlueCurve {keys [<at 0 1>,<at 0.7243 0.4125>]}
       GreenCurve {keys [<at 0 0.0137>,<at 0.7243 0.0056>]}
       RedCurve {keys [<at 0 0.402>,<at 0.7243 0.1651>]}
      }})
    R1 Saffron:LinearColor=&ColorFromCurve(ColorCurve=DataInterface() {propertyValues {
       AlphaCurve {keys [<at 0 3 RCIM_Cubic>,<at 0.8722 0.5>]}
       BlueCurve {keys [<at 0.0075 0>,<at 0.7226 0>]}
       GreenCurve {keys [<at 0.0075 0.9647>,<at 0.7226 0.5029>]}
       RedCurve {keys [<at 0.0075 1>,<at 0.7226 0.521>]}
      }})
    R1 SkyBlue:LinearColor=&ColorFromCurve(ColorCurve=DataInterface() {propertyValues {
       AlphaCurve {keys [<at 0 3 RCIM_Cubic>,<at 0.9004 0.5>]}
       BlueCurve {keys [<at 0 0.9647>,<at 0.7226 0.6038>]}
       GreenCurve {keys [<at 0 0.6724>,<at 0.7226 0.4125>]}
       RedCurve {keys [<at 0 0.0762>,<at 0.7226 0.016>]}
      }})
    R1 TitaniumWhite:LinearColor=&ColorFromCurve(ColorCurve=DataInterface() {propertyValues {
       AlphaCurve {keys [<at 0 3 RCIM_Cubic>,<at 1 0.5 RCIM_Cubic>]}
       BlueCurve {keys [<at 0 1>,<at 0.5306 0.7157>,<at 0.7199 0.5776>]}
       GreenCurve {keys [<at 0 1>,<at 0.5306 0.7157>,<at 0.7199 0.5776>]}
       RedCurve {keys [<at 0 1>,<at 0.5306 0.7157>,<at 0.7199 0.5776>]}
      }})
    R2 Black:LinearColor=<rgba 0 0.7991 0.5089 1>
    R2 BurntSienna:LinearColor=<rgba 0 0.7991 0.5089 1>
    R2 Cobalt:LinearColor=<rgba 0 0.7991 0.5089 1>
    R2 Crimson:LinearColor=<rgba 0 0.7991 0.5089 1>
    R2 DefaultAlpha=0.1
    R2 DefaultColor:LinearColor=&ColorFromCurve(ColorCurve=DataInterface() {propertyValues {
       AlphaCurve {keys [<at 0.0449 3>,<at 0.4797 3>,<at 1 3>]}
       BlueCurve {keys [<at 0.4885 0.5089>]}
       GreenCurve {keys [<at 0.4885 0.7991>]}
       RedCurve {keys [<at 0.4885 0>]}
      }})
    R2 ForestGreen:LinearColor=<rgba 0 0.7991 0.5089 1>
    R2 Grey:LinearColor=<rgba 0 0.7991 0.5089 1>
    R2 Lime:LinearColor=<rgba 0 0.7991 0.5089 1>
    R2 Orange:LinearColor=<rgba 0 0.7991 0.5089 1>
    R2 Pink:LinearColor=<rgba 0 0.7991 0.5089 1>
    R2 Purple:LinearColor=<rgba 0 0.7991 0.5089 1>
    R2 Saffron:LinearColor=<rgba 0 0.7991 0.5089 1>
    R2 SkyBlue:LinearColor=<rgba 0 0.7991 0.5089 1>
    R2 TitaniumWhite:LinearColor=<rgba 0 0.7991 0.5089 1>
    R3 Black:LinearColor=<rgba 0.624 0 0.0194 0.8196>
    R3 BurntSienna:LinearColor=<rgba 0.624 0 0.0194 0.8196>
    R3 Cobalt:LinearColor=<rgba 0.624 0 0.0194 0.8196>
    R3 Crimson:LinearColor=<rgba 0.624 0 0.0194 0.8196>
    R3 DefaultAlpha=4
    R3 DefaultColor:LinearColor=<rgba 0.624 0 0.0194 0.8196>
    R3 ForestGreen:LinearColor=<rgba 0.624 0 0.0194 0.8196>
    R3 Grey:LinearColor=<rgba 0.624 0 0.0194 0.8196>
    R3 Lime:LinearColor=<rgba 0.624 0 0.0194 0.8196>
    R3 Orange:LinearColor=<rgba 0.624 0 0.0194 0.8196>
    R3 Pink:LinearColor=<rgba 0.624 0 0.0194 0.8196>
    R3 Purple:LinearColor=<rgba 0.624 0 0.0194 0.8196>
    R3 Saffron:LinearColor=<rgba 0.624 0 0.0194 0.8196>
    R3 SkyBlue:LinearColor=<rgba 0.624 0 0.0194 0.8196>
    R3 TitaniumWhite:LinearColor=<rgba 0.624 0 0.0194 0.8196>
    R4 Black:LinearColor=&ColorFromCurve(ColorCurve=DataInterface() {propertyValues {
       AlphaCurve {keys [<at 0.2984 1 RCIM_Cubic>,<at 1 1 RCIM_Cubic>]}
       BlueCurve {keys [<at 0.0614 0.2831 RCIM_Cubic>,<at 1 0.1095 RCIM_Cubic>]}
       GreenCurve {keys [<at 0.0614 0.2831 RCIM_Cubic>,<at 1 0.1095 RCIM_Cubic>]}
       RedCurve {keys [<at 0.0614 0.2831 RCIM_Cubic>,<at 1 0.1095 RCIM_Cubic>]}
      }})
    R4 BurntSienna:LinearColor=&ColorFromCurve(ColorCurve=DataInterface() {propertyValues {
       AlphaCurve {keys [<at 0 4 RCIM_Cubic>,<at 1 0.67>]}
       BlueCurve {keys [<at 0 0 RCIM_Cubic>,<at 1 0 RCIM_Cubic>]}
       GreenCurve {keys [<at 0 0.2086 RCIM_Cubic>,<at 1 0.0742 RCIM_Cubic>]}
       RedCurve {keys [<at 0 0.3372 RCIM_Cubic>,<at 1 0.1779 RCIM_Cubic>]}
      }})
    R4 Cobalt:LinearColor=&ColorFromCurve(ColorCurve=DataInterface() {propertyValues {
       AlphaCurve {keys [<at 0.3816 1 RCIM_Cubic>,<at 0.9313 0.2 RCIM_Cubic>]}
       BlueCurve {keys [<at 0.1601 1 RCIM_Cubic>,<at 1 0.3095 RCIM_Cubic>]}
       GreenCurve {keys [<at 0.1601 0.0529 RCIM_Cubic>,<at 1 0.0021 RCIM_Cubic>]}
       RedCurve {keys [<at 0.1601 0.0103 RCIM_Cubic>,<at 1 0 RCIM_Cubic>]}
      }})
    R4 Crimson:LinearColor=&ColorFromCurve(ColorCurve=DataInterface() {propertyValues {
       AlphaCurve {keys [<at 0 1 RCIM_Cubic>,<at 1 1 RCIM_Cubic>]}
       BlueCurve {keys [<at 0 0.016 RCIM_Cubic>,<at 0.7239 0.0052 RCIM_Cubic>]}
       GreenCurve {keys [<at 0 0.016 RCIM_Cubic>,<at 0.7239 0.0052 RCIM_Cubic>]}
       RedCurve {keys [<at 0 1 RCIM_Cubic>,<at 0.7239 0.4851 RCIM_Cubic>]}
      }})
    R4 DefaultColor:LinearColor=&ColorFromCurve(ColorCurve=DataInterface() {propertyValues {
       AlphaCurve {keys [<at 0 1 RCIM_Cubic>,<at 1 0.74 RCIM_Cubic>]}
       BlueCurve {keys [<at 0.2632 0.5391 RCIM_Cubic>,<at 0.4445 0.9249>,<at 0.5131 0.9323>,<at 0.8845 2 RCIM_Cubic>]}
       GreenCurve {keys [<at 0.2632 1.431 RCIM_Cubic>,<at 0.4445 0.98>,<at 0.5131 0.8777>,<at 0.8845 0.3046 RCIM_Cubic>]}
       RedCurve {keys [<at 0.2632 1.5 RCIM_Cubic>,<at 0.4445 0.6584>,<at 0.5131 0>,<at 0.8845 0.0328 RCIM_Cubic>]}
      }})
    R4 ForestGreen:LinearColor=&ColorFromCurve(ColorCurve=DataInterface() {propertyValues {
       AlphaCurve {keys [<at 0 1 RCIM_Cubic>,<at 1 1 RCIM_Cubic>]}
       BlueCurve {keys [<at 0 0.1119 RCIM_Cubic>,<at 1 0.0296 RCIM_Cubic>]}
       GreenCurve {keys [<at 0 0.7991 RCIM_Cubic>,<at 1 0.1878 RCIM_Cubic>]}
       RedCurve {keys [<at 0 0 RCIM_Cubic>,<at 1 0.0033 RCIM_Cubic>]}
      }})
    R4 Grey:LinearColor=&ColorFromCurve(ColorCurve=DataInterface() {propertyValues {
       AlphaCurve {keys [<at 0 1 RCIM_Cubic>,<at 1 1 RCIM_Cubic>]}
       BlueCurve {keys [<at 0 0.4452 RCIM_Cubic>,<at 1 0.0723 RCIM_Cubic>]}
       GreenCurve {keys [<at 0 0.4452 RCIM_Cubic>,<at 1 0.0723 RCIM_Cubic>]}
       RedCurve {keys [<at 0 0.4452 RCIM_Cubic>,<at 1 0.0723 RCIM_Cubic>]}
      }})
    R4 Lime:LinearColor=&ColorFromCurve(ColorCurve=DataInterface() {propertyValues {
       AlphaCurve {keys [<at 0 2 RCIM_Cubic>,<at 1 1 RCIM_Cubic>]}
       BlueCurve {keys [<at 0 0.159 RCIM_Cubic>,<at 1 0.0452 RCIM_Cubic>]}
       GreenCurve {keys [<at 0 1 RCIM_Cubic>,<at 1 0.2705 RCIM_Cubic>]}
       RedCurve {keys [<at 0 0.6867 RCIM_Cubic>,<at 1 0.1714 RCIM_Cubic>]}
      }})
    R4 Orange:LinearColor=&ColorFromCurve(ColorCurve=DataInterface() {propertyValues {
       AlphaCurve {keys [<at 0 2 RCIM_Cubic>,<at 1 1 RCIM_Cubic>]}
       BlueCurve {keys [<at 0 0.004 RCIM_Cubic>,<at 1 0.0176 RCIM_Cubic>]}
       GreenCurve {keys [<at 0 0.3813 RCIM_Cubic>,<at 1 0.2086 RCIM_Cubic>]}
       RedCurve {keys [<at 0 0.807 RCIM_Cubic>,<at 1 0.4851 RCIM_Cubic>]}
      }})
    R4 Pink:LinearColor=&ColorFromCurve(ColorCurve=DataInterface() {propertyValues {
       AlphaCurve {keys [<at 0 1 RCIM_Cubic>,<at 1 1 RCIM_Cubic>]}
       BlueCurve {keys [<at 0 0.7379 RCIM_Cubic>,<at 1 0.305 RCIM_Cubic>]}
       GreenCurve {keys [<at 0 0.3231 RCIM_Cubic>,<at 1 0.0356 RCIM_Cubic>]}
       RedCurve {keys [<at 0 0.9823 RCIM_Cubic>,<at 1 0.4621 RCIM_Cubic>]}
      }})
    R4 Purple:LinearColor=&ColorFromCurve(ColorCurve=DataInterface() {propertyValues {
       AlphaCurve {keys [<at 0 1 RCIM_Cubic>,<at 1 1 RCIM_Cubic>]}
       BlueCurve {keys [<at 0 1 RCIM_Cubic>,<at 1 0.4793 RCIM_Cubic>]}
       GreenCurve {keys [<at 0 0 RCIM_Cubic>,<at 1 0 RCIM_Cubic>]}
       RedCurve {keys [<at 0 0.4125 RCIM_Cubic>,<at 1 0.1981 RCIM_Cubic>]}
      }})
    R4 Saffron:LinearColor=&ColorFromCurve(ColorCurve=DataInterface() {propertyValues {
       AlphaCurve {keys [<at 0 1 RCIM_Cubic>,<at 1 1 RCIM_Cubic>]}
       BlueCurve {keys [<at 0 0.2502 RCIM_Cubic>,<at 1 0.0252 RCIM_Cubic>]}
       GreenCurve {keys [<at 0 1 RCIM_Cubic>,<at 1 0.5457 RCIM_Cubic>]}
       RedCurve {keys [<at 0 1 RCIM_Cubic>,<at 1 0.5457 RCIM_Cubic>]}
      }})
    R4 SkyBlue:LinearColor=&ColorFromCurve(ColorCurve=DataInterface() {propertyValues {
       AlphaCurve {keys [<at 0 1.6 RCIM_Cubic>,<at 1 1 RCIM_Cubic>]}
       BlueCurve {keys [<at 0 1 RCIM_Cubic>,<at 1 0.4397 RCIM_Cubic>]}
       GreenCurve {keys [<at 0 0.6867 RCIM_Cubic>,<at 1 0.3005 RCIM_Cubic>]}
       RedCurve {keys [<at 0 0.0369 RCIM_Cubic>,<at 1 0.016 RCIM_Cubic>]}
      }})
    R4 TitaniumWhite:LinearColor=&ColorFromCurve(ColorCurve=DataInterface() {propertyValues {
       AlphaCurve {keys [<at 0 1 RCIM_Cubic>,<at 1 1 RCIM_Cubic>]}
       BlueCurve {keys [<at 0 1.3 RCIM_Cubic>,<at 1 0.2918 RCIM_Cubic>]}
       GreenCurve {keys [<at 0 1.3 RCIM_Cubic>,<at 1 0.2918 RCIM_Cubic>]}
       RedCurve {keys [<at 0 1.3 RCIM_Cubic>,<at 1 0.2918 RCIM_Cubic>]}
      }})
    Use R1 Painted Overrides=true
    Use R2 Painted Overrides=true
    Use R3 Painted Overrides=true
    Use R4 Painted Overrides=true
   ) {module=VC_PaintedVariantsOverride002}
  ]
  rendererClasses [&NiagaraRibbonRendererProperties]
  renderers [
   {rendererClass=&NiagaraRibbonRendererProperties}
   {rendererClass=&NiagaraRibbonRendererProperties, rendererIndex=1}
   {rendererClass=&NiagaraRibbonRendererProperties, rendererIndex=2}
   {rendererClass=&NiagaraRibbonRendererProperties, rendererIndex=3}
  ]
 }
 Traingles_Chromatic {
  emitterUpdateScript [
   &EmitterState()
   &SpawnRate(
    SpawnRate=&Multiply_Float(
     A=$Engine.Emitter.SpawnCountScale
     B=&VC_Mask_Floats(Coro Value:NiagaraFloat=25, FrontEnd Value:NiagaraFloat=0.5, FN Value:NiagaraFloat=0.5)
    )
    Use Spawn Probability=false
   ) {module=SpawnRate001}
   &VC_RateQualityHandler(Epic Spawn Rate=40, High Spawn Rate=40, Low Spawn Rate=40, Medium Spawn Rate=40, Use Custom Rate?=true)
   &VC_SpawnRibbon_v1(Number of Ribbons=2, Spawn Time=-1, SpawnRate=$Emitter.Spawn Rate)
  ]
  particleSpawnScript [
   &VC_PaintedVariantsOverride() {enabled=false}
   &VC_InitializeRibbon_v1(
    Color=$Particles.Color
    Custom Offset=true
    Enable Skeletal Position=true
    Lifetime=&MergeHelper_FloatSwitch(0=1.1, 1=1.1, 2=1.1, 3=1.1, Index=$User.GameModeID, Num=::NewEnumerator2)
    Local Vector Offset=&AddVectorToPosition(Position=<-30 0 -30>, Vector=Linked() {linkedVariable {name=User.DTCarOffset, type=NiagaraPosition}})
    Offset Amount=1
    Ribbon Width=55
   )
   &AddVelocity(Velocity=&MergeHelper_VectorSwitch(0=<-1000 0 0>, 1=<-1100 0 0>, 2=<-1000 0 0>, 3=<-1000 0 0>, Index=0, Num=::NewEnumerator2)) {module=AddVelocity001}
   &DLR_ProductPreviews(GameModeID:NiagaraInt32=$System.GameModeID, bPreviewVehicleSpeed:NiagaraBool=$System.bPreviewVehicleSpeed)
  ]
  particleUpdateScript [
   &FN_DisableTireFX(Mask Index Attribute:NiagaraInt32=$Particles.SpawnGroup, Mask Index FL=2, Mask Index FR=3, Mask Index RL=0, Mask Index RR=1)
   &InheritSourceMovement(
    Applied Position Scale=&Multiply_VectorByFloat(Vector=<1 1 1>, Float=1)
    ApplyForce=false
    ApplyPosition=true
    ApplyPositionFalloff=false
   )
   &ParticleState()
   &InheritSourceMovement(
    Applied Position Scale=&Multiply_VectorByFloat(
     Vector=<1 1 1>
     Float=&FloatFromCurve(FloatCurve=DataInterface() {propertyValues {Curve {
         defaultValue=1
         keys [
          {arriveTangent=-1.4003, interpMode=RCIM_Cubic, leaveTangent=-1.4003, tangentMode=RCTM_User, time=0, value=1}
          {arriveTangent=-0.1632, interpMode=RCIM_Cubic, leaveTangent=-0.1632, tangentMode=RCTM_User, time=1, value=0}
         ]
        }}})
    )
    ApplyForce=false
    ApplyPosition=true
    ApplyPositionFalloff=false
   ) {module=InheritSourceMovement001}
   &SolveForcesAndVelocity(Clamp Velocity=false, Limit Acceleration=false)
   &VC_UpdateRibbon_v1(Enable Skeletal Position=true)
   &VC_PaintedVariantsOverride(
    Orange=<rgba 0.6724 0.0883 0.2836 1>
    R1 Black:LinearColor=&ColorFromCurve(ColorCurve=DataInterface() {propertyValues {
       AlphaCurve {keys [<at 0 4 RCIM_Cubic>,<at 0.1045 0.9695>,<at 1 0 RCIM_Cubic>]}
       BlueCurve {keys [<at 0 0.0356 RCIM_Cubic>,<at 0.6884 0.0212>]}
       GreenCurve {keys [<at 0 0.0356 RCIM_Cubic>,<at 0.6884 0.0212>]}
       RedCurve {keys [<at 0 0.0356 RCIM_Cubic>,<at 0.6884 0.0212>]}
      }})
    R1 BurntSienna:LinearColor=&ColorFromCurve(ColorCurve=DataInterface() {propertyValues {
       AlphaCurve {keys [<at 0 4 RCIM_Cubic>,<at 0.2166 2>,<at 1 0 RCIM_Cubic>]}
       BlueCurve {keys [<at 0 0.0284>,<at 0.6066 0.006>]}
       GreenCurve {keys [<at 0 0.0723>,<at 0.6066 0.0194>]}
       RedCurve {keys [<at 0 0.2307>,<at 0.6066 0.0666>]}
      }})
    R1 Cobalt:LinearColor=&ColorFromCurve(ColorCurve=DataInterface() {propertyValues {
       AlphaCurve {keys [<at 0 4 RCIM_Cubic>,<at 0.6217 1>,<at 1 0 RCIM_Cubic>]}
       BlueCurve {keys [<at 0 0.6867 RCIM_Cubic>,<at 0.6282 0.0666>]}
       GreenCurve {keys [<at 0 0.0551 RCIM_Cubic>,<at 0.6282 0.0012>]}
       RedCurve {keys [<at 0 0.0429 RCIM_Cubic>,<at 0.6282 0>]}
      }})
    R1 Crimson:LinearColor=&ColorFromCurve(ColorCurve=DataInterface() {propertyValues {
       AlphaCurve {keys [<at 0 4 RCIM_Cubic>,<at 1 0 RCIM_Cubic>]}
       BlueCurve {keys [<at 0.0162 0.0666>,<at 0.5441 0.007>]}
       GreenCurve {keys [<at 0.0162 0.0545>,<at 0.5441 0>]}
       RedCurve {keys [<at 0.0162 0.5271>,<at 0.5441 0.2705>]}
      }})
    R1 DefaultAlpha=0.9
    R1 DefaultColor=&ColorFromCurve(ColorCurve=DataInterface() {propertyValues {
       AlphaCurve {keys [<at 0 4 RCIM_Cubic>,<at 0.4519 2.369>,<at 0.8546 0.9976>,<at 1 0.248 RCIM_Cubic>]}
       BlueCurve {keys [<at 0.2825 1>,<at 0.4849 0.4037>,<at 0.631 1>,<at 0.7138 0.1771>]}
       GreenCurve {keys [<at 0.2825 0.947>,<at 0.4849 1>,<at 0.631 0.9298>,<at 0.7138 0.0449>]}
       RedCurve {keys [<at 0.2825 0.9293>,<at 0.4849 0.0023>,<at 0.631 0>,<at 0.7138 0>]}
      }})
    R1 ForestGreen:LinearColor=&ColorFromCurve(ColorCurve=DataInterface() {propertyValues {
       AlphaCurve {keys [<at 0 4 RCIM_Cubic>,<at 1 0 RCIM_Cubic>]}
       BlueCurve {keys [<at 0.0075 0.0723>,<at 0.5485 0.0091>]}
       GreenCurve {keys [<at 0.0075 0.3419>,<at 0.5485 0.0666>]}
       RedCurve {keys [<at 0.0075 0.0578>,<at 0.5485 0.006>]}
      }})
    R1 Gold:LinearColor=&ColorFromCurve(ColorCurve=DataInterface() {propertyValues {
       AlphaCurve {keys [<at 0 1 RCIM_Cubic>,<at 1 0 RCIM_Cubic>]}
       BlueCurve {keys [<at 0 0.063 RCIM_Cubic>,<at 0.953 0.0342 RCIM_Cubic>]}
       GreenCurve {keys [<at 0 0.4397 RCIM_Cubic>,<at 0.953 0.2389 RCIM_Cubic>]}
       RedCurve {keys [<at 0 0.6038 RCIM_Cubic>,<at 0.953 0.3281 RCIM_Cubic>]}
      }})
    R1 Grey:LinearColor=&ColorFromCurve(ColorCurve=DataInterface() {propertyValues {
       AlphaCurve {keys [<at 0 1 RCIM_Cubic>,<at 0.8868 0.9>,<at 0.9795 0 RCIM_Cubic>]}
       BlueCurve {keys [<at 0 0.2527>,<at 0.5743 0.0666>]}
       GreenCurve {keys [<at 0 0.2527>,<at 0.5743 0.0666>]}
       RedCurve {keys [<at 0 0.2527>,<at 0.5743 0.0666>]}
      }})
    R1 Lime:LinearColor=&ColorFromCurve(ColorCurve=DataInterface() {propertyValues {
       AlphaCurve {keys [<at 0 1 RCIM_Cubic>,<at 0.8997 0.9>,<at 1 0 RCIM_Cubic>]}
       BlueCurve {keys [<at 0.014 0.0508>,<at 0.611 0.0331>]}
       GreenCurve {keys [<at 0.014 0.6435>,<at 0.611 0.3613>]}
       RedCurve {keys [<at 0.014 0.344>,<at 0.611 0.1946>]}
      }})
    R1 Orange:LinearColor=&ColorFromCurve(ColorCurve=DataInterface() {propertyValues {
       AlphaCurve {keys [<at 0 1 RCIM_Cubic>,<at 0.9213 0.986>,<at 1 0 RCIM_Cubic>]}
       BlueCurve {keys [<at 0 0.0705 RCIM_Cubic>,<at 0.6217 0.0075>]}
       GreenCurve {keys [<at 0 0.357 RCIM_Cubic>,<at 0.6217 0.1329>]}
       RedCurve {keys [<at 0 0.892 RCIM_Cubic>,<at 0.6217 0.3712>]}
      }})
    R1 Pink:LinearColor=&ColorFromCurve(ColorCurve=DataInterface() {propertyValues {
       AlphaCurve {keys [<at 0 1 RCIM_Cubic>,<at 0.9277 0.9949>,<at 1 0 RCIM_Cubic>]}
       BlueCurve {keys [<at 0.0334 0.6486>,<at 0.6282 0.2874>]}
       GreenCurve {keys [<at 0.0334 0.154>,<at 0.6282 0.0395>]}
       RedCurve {keys [<at 0.0334 0.9323>,<at 0.6282 0.4342>]}
      }})
    R1 Purple:LinearColor=&ColorFromCurve(ColorCurve=DataInterface() {propertyValues {
       AlphaCurve {keys [<at 0 1 RCIM_Cubic>,<at 0.8976 0.9795>,<at 1 0 RCIM_Cubic>]}
       BlueCurve {keys [<at 0.042 0.552>,<at 0.6239 0.1046>]}
       GreenCurve {keys [<at 0.042 0.0436>,<at 0.6239 0.0097>]}
       RedCurve {keys [<at 0.042 0.2523>,<at 0.6239 0.0482>]}
      }})
    R1 Saffron:LinearColor=&ColorFromCurve(ColorCurve=DataInterface() {propertyValues {
       AlphaCurve {keys [<at 0 1 RCIM_Cubic>,<at 0.9083 0.9979>,<at 1 0 RCIM_Cubic>]}
       BlueCurve {keys [<at 0.0075 0.2033>,<at 0.6174 0.0168>]}
       GreenCurve {keys [<at 0.0075 0.7769>,<at 0.6174 0.2542>]}
       RedCurve {keys [<at 0.0075 0.952>,<at 0.6174 0.3278>]}
      }})
    R1 SkyBlue:LinearColor=&ColorFromCurve(ColorCurve=DataInterface() {propertyValues {
       AlphaCurve {keys [<at 0 1 RCIM_Cubic>,<at 0.9428 0.9815>,<at 1 0 RCIM_Cubic>]}
       BlueCurve {keys [<at 0 0.7396>,<at 0.6088 0.2961>]}
       GreenCurve {keys [<at 0 0.5137>,<at 0.6088 0.2051>]}
       RedCurve {keys [<at 0 0.0584>,<at 0.6088 0.011>]}
      }})
    R1 TitaniumWhite:LinearColor=&ColorFromCurve(ColorCurve=DataInterface() {propertyValues {
       AlphaCurve {keys [<at 0 1 RCIM_Cubic>,<at 1 0 RCIM_Cubic>]}
       BlueCurve {keys [<at 0 0.9583>,<at 0.5808 0.6795>]}
       GreenCurve {keys [<at 0 0.9583>,<at 0.5808 0.6795>]}
       RedCurve {keys [<at 0 0.9583>,<at 0.5808 0.6795>]}
      }})
    R2 Black:LinearColor=<rgba 0 0.5972 0.3763 1>
    R2 BurntSienna:LinearColor=<rgba 0 0.5972 0.3763 1>
    R2 Cobalt:LinearColor=<rgba 0 0.5972 0.3763 1>
    R2 Crimson:LinearColor=<rgba 0 0.5972 0.3763 1>
    R2 DefaultAlpha=0.8
    R2 DefaultColor:LinearColor=&ColorFromCurve(ColorCurve=DataInterface() {propertyValues {
       AlphaCurve {keys [<at 0.0449 3>,<at 0.4797 3>,<at 1 3>]}
       BlueCurve {keys [<at 0.4885 0.3763>]}
       GreenCurve {keys [<at 0.4885 0.5972>]}
       RedCurve {keys [<at 0.4885 0>]}
      }})
    R2 ForestGreen:LinearColor=<rgba 0 0.5972 0.3763 1>
    R2 Grey:LinearColor=<rgba 0 0.5972 0.3763 1>
    R2 Lime:LinearColor=<rgba 0 0.5972 0.3763 1>
    R2 Orange:LinearColor=<rgba 0 0.5972 0.3763 1>
    R2 Pink:LinearColor=<rgba 0 0.5972 0.3763 1>
    R2 Purple:LinearColor=<rgba 0 0.5972 0.3763 1>
    R2 Saffron:LinearColor=<rgba 0 0.5972 0.3763 1>
    R2 SkyBlue:LinearColor=<rgba 0 0.5972 0.3763 1>
    R2 TitaniumWhite:LinearColor=<rgba 0 0.5972 0.3763 1>
    R3 Black:LinearColor=<rgba 0.3095 0 0.0097 0.8196>
    R3 BurntSienna:LinearColor=<rgba 0.3095 0 0.0097 0.8196>
    R3 Cobalt:LinearColor=<rgba 0.3095 0 0.0097 0.8196>
    R3 Crimson:LinearColor=<rgba 0.3095 0 0.0097 0.8196>
    R3 DefaultAlpha=4
    R3 DefaultColor:LinearColor=<rgba 0.3095 0 0.0097 0.8196>
    R3 ForestGreen:LinearColor=<rgba 0.3095 0 0.0097 0.8196>
    R3 Grey:LinearColor=<rgba 0.3095 0 0.0097 0.8196>
    R3 Lime:LinearColor=<rgba 0.3095 0 0.0097 0.8196>
    R3 Orange:LinearColor=<rgba 0.3095 0 0.0097 0.8196>
    R3 Pink:LinearColor=<rgba 0.3095 0 0.0097 0.8196>
    R3 Purple:LinearColor=<rgba 0.3095 0 0.0097 0.8196>
    R3 Saffron:LinearColor=<rgba 0.3095 0 0.0097 0.8196>
    R3 SkyBlue:LinearColor=<rgba 0.3095 0 0.0097 0.8196>
    R3 TitaniumWhite:LinearColor=<rgba 0.3095 0 0.0097 0.8196>
    Use R1 Painted Overrides=true
    Use R2 Painted Overrides=true
    Use R3 Painted Overrides=true
   ) {module=VC_PaintedVariantsOverride002}
  ]
  rendererClasses [&NiagaraRibbonRendererProperties]
  renderers [
   {rendererClass=&NiagaraRibbonRendererProperties}
   {rendererClass=&NiagaraRibbonRendererProperties, rendererIndex=1}
   {rendererClass=&NiagaraRibbonRendererProperties, rendererIndex=2}
  ]
 }
}
summary {
 systemName=NS_DriftTrail_ChromaticAberration
 emitters [
  {emitterName=Trail_Chromatic, rendererClasses [&NiagaraRibbonRendererProperties]}
  {emitterName=Traingles_Chromatic, rendererClasses [&NiagaraRibbonRendererProperties]}
  {emitterName=NE_WheelWrap, rendererClasses [&NiagaraMeshRendererProperties]}
  {emitterName=NE_WheelWrap001, rendererClasses [&NiagaraMeshRendererProperties]}
 ]
 userVariables [
  User.bIsFrontendPreview {type=NiagaraBool}
  User.bIsPainted {type=NiagaraBool}
  User.blsIsCoroColor {type=NiagaraBool}
  User.bPreviewVehicleSpeed {type=NiagaraBool}
  User.Brightness2 {type=NiagaraFloat, defaultValue=1, description=""}
  User.ColorVariantID {type=NiagaraInt32}
  User.ConeColor {type=LinearColor, description="", defaultValue {struct=LinearColor, value=<rgba 1 1 1 1>}}
  User.DTCarOffset {type=NiagaraPosition, description="", defaultValue {}}
  User.DTCarOffset002 {type=NiagaraPosition, description="", defaultValue {}}
  User.GameModeID {type=NiagaraInt32}
  User.MultiVariantColorCone {type=NiagaraFloat, defaultValue=2.5, description=""}
  User.offset {type=NiagaraPosition, description="", defaultValue {}}
  User.offset001 {type=Vector3f, description="", defaultValue {struct=Vector3f, value=<0 0 0>}}
  User.offset002 {type=Vector3f, description="", defaultValue {struct=Vector3f, value=<0 0 0>}}
  User.PaintedColor {type=LinearColor, description="", defaultValue {struct=LinearColor, value=<rgba 1 1 1 1>}}
  User.PaintedColorTriangles {type=LinearColor, description="", defaultValue {struct=LinearColor, value=<rgba 1 1 1 1>}}
  User.PaintedScalar {type=NiagaraFloat, defaultValue=1, description=""}
  User.PowerTriColorPaintedCone {type=NiagaraFloat, defaultValue=1.5, description=""}
  User.VehicleFlags {type=NiagaraInt32, defaultValue=806289408}
  User.Wheel_Radius_Array {
   type=NiagaraDataInterfaceArrayFloat
   defaultValue {
    struct=NiagaraExt_VariableValue_DataInterface
    value {
     dataInterface=&self:NiagaraDataInterfaceArrayFloat_0
     dataInterfaceClass=&NiagaraDataInterfaceArrayFloat
    }
   }
  }
  User.Wheel_Width_Array {
   type=NiagaraDataInterfaceArrayFloat
   defaultValue {
    struct=NiagaraExt_VariableValue_DataInterface
    value {
     dataInterface=&self:NiagaraDataInterfaceArrayFloat_1
     dataInterfaceClass=&NiagaraDataInterfaceArrayFloat
    }
   }
  }
  User.WheelSize {type=NiagaraFloat, defaultValue=0.1, description=""}
 ]
}
